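# Unlit translucent Niagara sprite material for ConfectionPop pickaxe effects, drawing an outlined graphic from a grayscale SDF texture.
# Two step thresholds on the SDF separate fill from outline, with LineThickness offsetting the second.
# A gradient texture tints both regions, while a static switch can substitute particle color for the fill.
# Time-of-day scalars scale the color before emissive adjustment, and particle alpha times the shape mask forms the alpha.
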Material /BRCosmetics/Weapons/FORT_Melee/Pickaxe_ConfectionPop/FX/M_ConfectionPop_Graphics
graph {
 properties {
  bCastRayTracedShadows=false
  bOutputTranslucentVelocity=true
  bUsedWithNiagaraSprites=true
  blendMode=BLEND_Translucent
  pixelDepthOffsetMode=PDOM_Legacy
  shadingModel=MSM_Unlit
  usedShadingModels=Unlit
 }
 nodes [
  n1 = ScalarParameter() {defaultValue=1, parameterName=Value_Morning}
  n2 = ScalarParameter() {defaultValue=1, parameterName=Value_Day}
  n3 = ScalarParameter() {defaultValue=1, parameterName=Value_Evening}
  n4 = ScalarParameter() {defaultValue=1, parameterName=Value_Night}
  n5 = VectorParameter() {parameterName=OutlineColor, defaultValue=<rgba 0.1864 1 0.9602 1>}
  n6 = VectorParameter() {parameterName=OutlineColorGradient, defaultValue=<rgba 0.1864 1 0.9602 1>}
  n7 = TextureSampleParameter2D() {
   parameterName=Gradient
   samplerType=SAMPLERTYPE_LinearGrayscale
   texture=&/T_BeepBoopGradient_M
  }
  n8 = Saturate(None=Power(Base=OneMinus(None=@n7.R), Exp=ScalarParameter() {defaultValue=1, parameterName=GradientPower}))
  n9 = ScalarParameter() {defaultValue=1, parameterName=GradientStenght}
  n10 = TextureSampleParameter2D() {parameterName=TextSDF, samplerType=SAMPLERTYPE_Grayscale, texture=&/T_BeepBoop_M}
  n11 = ScalarParameter() {defaultValue=0.2, parameterName=Upper Limit}
  n12 = Step(Y=@n11, X=@n10.R) {constX=0.1, constY=0.2}
  n13 = ScalarParameter() {defaultValue=0.2, parameterName=LineThickness, sliderMax=1}
  n14 = Constant()
  n15 = Constant() {r=2}
  n16 = Step(
   Y=Add(
    A=@n11
    B=MaterialFunctionCall(Input=@n13, Input Low=@n14, Input High=@n15) {materialFunction=&RemapValue_0-1}.Result
   )
   X=@n10.R
  ) {constX=0.1, constY=0.2}
  n17 = Multiply(
   A=LinearInterpolate(A=@n5.RGB, B=LinearInterpolate(A=@n5.RGB, B=@n6.RGB, Alpha=@n8), Alpha=@n9)
   B=LinearInterpolate(A=@n12, Alpha=@n16) {constB=0}
  )
  n18 = ParticleColor()
  n19 = VectorParameter() {parameterName=MainColor, defaultValue=<rgba 0.5417 0.1092 0.2182 1>}
  n20 = LinearInterpolate(A=@n19.RGB, B=VectorParameter() {parameterName=GradientColor, defaultValue=<rgba 0.5417 0.5417 0.5417 1>}.RGB, Alpha=@n8)
  n21 = Multiply(A=StaticSwitchParameter(True=@n18.RGB, False=LinearInterpolate(A=@n19.RGB, B=@n20, Alpha=@n9)) {parameterName=UseParticleColor}, B=@n12)
  MaterialFunctionCall(Emissive=Multiply(
    A=MaterialFunctionCall(Value_Morning=@n1, Value_Day=@n2, Value_Evening=@n3, Value_Night=@n4) {materialFunction=&Scalar_Time_of_Day}.Result
    B=LinearInterpolate(A=@n17, B=@n21, Alpha=@n16)
   )) {materialFunction=&MF_FX_EmissiveAdjustments}
  Multiply(A=@n18.A, B=@n12)
 ]
}
