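# Niagara swing trail effect for the Confection Pop pickaxe cosmetic.
# A single Trail emitter uses AnimTrail modules to spawn, initialize and update short-lived ribbon particles from bone or socket indices.
# Each update writes separate ToonColor and NonToonColor attributes, tinted by time of day with curve-driven alpha fades, and curve-driven dynamic material parameters.
# Two ribbon renderers draw the emitter, and user variables include toon flags and a skeletal mesh.
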NiagaraSystem /BRCosmetics/Weapons/FORT_Melee/Pickaxe_ConfectionPop/FX/NS_ConfectionPop_Trail
dependencies {usedDataInterfaces [&NiagaraDataInterfaceCurve], usedDynamicInputs [], usedModules [&SystemState]}
emitters {Trail {
  emitterUpdateScript [
   &EmitterState(Life Cycle Mode=::NewEnumerator1, Loop Behavior=::NewEnumerator1, Loop Duration=0.25, UseLoopDelay=true)
   &FN_SpawnAnimTrail_v2()
  ]
  particleSpawnScript [
   /Game/Effects/Niagara/Modules/AnimTrail/FN_InitializeAnimTrail/FN_InitializeAnimTrailBoneIndecies(Second Bone or Socket Index=1)
   /Game/Effects/Niagara/Modules/AnimTrail/FN_InitializeAnimTrail/FN_InitializeAnimTrail_v3/FN_InitializeAnimTrail(Color:LinearColor=<rgba 0 0 0 10>, Lifetime=0.12)
   &AddVelocity(Velocity=$User.Debug)
  ]
  particleUpdateScript [
   /Game/Effects/Niagara/Modules/AnimTrail/FN_UpdateAnimTrail/FN_UpdateAnimTrail_v3/FN_UpdateAnimTrail(Locked Particle RibbonWidth:NiagaraFloat=$Particles.RibbonWidth)
   &SolveForcesAndVelocity(Clamp Velocity=false, Limit Acceleration=false)
   &DynamicMaterialParameters(
    Index 0 Param 1=&FloatFromCurve(
     FloatCurve=DataInterface() {propertyValues {Curve {keys [{tangentMode=RCTM_User, time=0, value=0}, {tangentMode=RCTM_User, time=1, value=1}]}}}
     Scale Curve=0.3
    )
    Module.Param0WriteEnabled=true
    Module.Param1WriteEnabled=true
    Module.Param2WriteEnabled=true
    Module.Param3WriteEnabled=true
   )
   /SetVariables(Particles.ToonColor:LinearColor=&MakeLinearColorFromVectorAndFloat(
     "Vector (RGB)"=&MakeVectorFromLinearColorRGB(Color=&FNTimeOfDayColor(
       Night Color=<rgba 0.1119 0.1779 0.2159 1>
       Morning Color=<rgba 0.5 0.375 0.4297 1>
       Day Color=<rgba 1 1 1 1>
       Evening Color=<rgba 0.5 0.3894 0.375 1>
      ))
     "Float (Alpha)"=&FloatFromCurve(
      FloatCurve=DataInterface() {propertyValues {Curve {keys [{arriveTangent=-2.5744, interpMode=RCIM_Cubic, leaveTangent=-2.5744, tangentMode=RCTM_User, time=0, value=1}, <at 1 0 RCIM_Cubic>]}}}
      Scale Curve=2
     )
    )) {isSetParameters=true}
   /SetVariables(Particles.NonToonColor:LinearColor=&MakeLinearColorFromVectorAndFloat(
     "Vector (RGB)"=&MakeVectorFromLinearColorRGB(Color=&MultiplyLinearColor(
       Linear Color=<rgba 1 1 1 1>
       Scale Factor=&MakeVector4_FromLinearColor(LinearColor=&FNTimeOfDayColor(
         Night Color=<rgba 0.2 0.3269 0.4 1>
         Morning Color=<rgba 0.502 0.3765 0.4314 1>
         Day Color=<rgba 1 1 1 1>
         Evening Color=<rgba 0.5 0.3894 0.375 1>
        ))
      ))
     "Float (Alpha)"=&FloatFromCurve(
      FloatCurve=DataInterface() {propertyValues {Curve {keys [
          {interpMode=RCIM_Cubic, tangentMode=RCTM_User, time=0, value=1}
          {arriveTangent=-1.8211, interpMode=RCIM_Cubic, leaveTangent=-1.8211, tangentMode=RCTM_User, time=1, value=0}
         ]}}}
      Scale Curve=1.5
     )
    )) {isSetParameters=true}
  ]
  rendererClasses [&NiagaraRibbonRendererProperties]
  renderers [
   {rendererClass=&NiagaraRibbonRendererProperties}
   {rendererClass=&NiagaraRibbonRendererProperties, rendererIndex=1}
  ]
 }}
summary {
 systemName=NS_ConfectionPop_Trail
 emitters [{emitterName=Trail, rendererClasses [&NiagaraRibbonRendererProperties]}]
 userVariables [
  User.ColorMultiplier {type=NiagaraFloat, defaultValue=7, description=""}
  User.Debug {type=Vector3f, description="", defaultValue {struct=Vector3f, value=<0 0 0>}}
  User.IsNotToon {type=NiagaraBool}
  User.IsToon {type=NiagaraBool}
  User.RibbonIDOffset {type=NiagaraInt32}
  User.SkeletalMesh {
   type=NiagaraDataInterfaceSkeletalMesh
   defaultValue {
    struct=NiagaraExt_VariableValue_DataInterface
    value {dataInterface=&self:NiagaraDataInterfaceSkeletalMesh_1}
   }
  }
  User.VariantNumber {type=NiagaraInt32}
 ]
}
